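# Control Rig for the F_MED_Firework face accessory part that applies a corrective offset to hair bones based on head orientation.
# A spherical pose reader on the head bone outputs a weight, passed through AlphaInterp with interpolation and clamping off.
# That weight blends a Modify Transforms offset on base bones, including right and center back hair, which parent the dyn_ chains.
# Debug drawing is left enabled on the pose reader.
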ControlRigBlueprint /Game/Characters/Player/Female/Medium/Bodies/F_MED_Firework/Meshes/Parts/F_MED_Firework_FaceAcc_CtrlRig
graphs [
 ContainedGraph {nodes [
   n1 = RigVMFunctionEntryNode()
   SequenceAggregate = &RigVMFunction_Sequence(ExecuteContext=@n1.ExecuteContext)
   SequenceAggregate_1 = &RigVMFunction_Sequence(ExecuteContext=@SequenceAggregate.B)
   RigVMFunctionReturnNode(A=@SequenceAggregate.A, B=@SequenceAggregate_1.A, C=@SequenceAggregate_1.B)
  ]}
 RigVMFunctionLibrary {links [], nodes []}
 RigVMModel {nodes [
   RigUnit_BeginExecution = &RigUnit_BeginExecution()
   SequenceAggregate = RigVMAggregateNode(ExecuteContext=@RigUnit_BeginExecution.ExecutePin)
   n1 = &RigUnit_SphericalPoseReader(
    ActiveRegionScaleFactors=(PositiveWidth=1,NegativeWidth=1,PositiveHeight=1,NegativeHeight=1)
    ActiveRegionSize=0.05
    Debug=(bDrawDebug=True,bDraw2D=False,bDrawLocalAxes=False,DebugScale=25,DebugSegments=20,DebugThickness=0.25)
    DriverAxis=(X=1,Y=0,Z=0)
    DriverItem=(Type=Bone,Name="head")
    FalloffRegionScaleFactors=(PositiveWidth=1,NegativeWidth=1,PositiveHeight=1,NegativeHeight=1)
    FalloffSize=0.18
    FlipHeightScaling=false
    FlipWidthScaling=false
    OptionalParentItem=(Type=Bone)
    RotationOffset=(X=70,Y=-20,Z=-125)
    ExecutePin=@SequenceAggregate.A
   )
   AlphaInterp = &RigVMFunction_AlphaInterp(InterpSpeedDecreasing=10, InterpSpeedIncreasing=10, Scale=1, bClampResult=False, bInterpResult=False, Value=@n1.OutputParam)
   > &RigUnit_ModifyTransforms(
    ItemToModify="((Item=(Type=Bone,Name=\"base_backHair_R\"),Transform=(Rotation=(X=-0.0019,Y=0.0436,Z=0.0436,W=0.9981),Translation=(X=0,Y=0,Z=0),Scale3D=(X=1,Y=1,Z=1))),(Item=(Type=Bone,Name=\"base_backHair_C\"),Transform=(Rotation=(X=0.1384,Y=0.4054,Z=-0.1656,W=0.8883),Trans ...(truncated)"
    Weight=@AlphaInterp.Result
   )
  ]}
]
hierarchy [
 root {loc=<0 0 0>, rot=<0 0 0 1>}
 pelvis {parent=root, loc=<0 -0.5301 98.0571>, rot=<0.0127 0.707 -0.0127 -0.707>}
 spine_01 {parent=pelvis, loc=<6.1208 -0.9328 0>, rot=<0 0 0.034 -0.9994>}
 spine_02 {parent=spine_01, loc=<6.5996 -0.2095 0>, rot=<0 0 0.0238 -0.9997>}
 spine_03 {parent=spine_02, loc=<6.4914 -0.1237 0>, rot=<0 0 0.0149 -0.9999>}
 spine_04 {parent=spine_03, loc=<6.9598 0.0794 0>, rot=<0 0 -0.0132 -0.9999>}
 spine_05 {parent=spine_04, loc=<6.9987 0 0>, rot=<0 0 -0.0015 -1>}
 neck_01 {parent=spine_05, loc=<11.3138 0.1529 0>, rot=<0 0 -0.2003 -0.9797>}
 neck_02 {parent=neck_01, loc=<4.5627 0 0>, rot=<0 0 0.0504 -0.9987>}
 head {parent=neck_02, loc=<5.3914 0 0>, rot=<0 0 0.115 -0.9934>}
 dyn_frontHair_L01 {parent=head, loc=<4.8626 6.4568 -5.5308>, rot=<0.3931 -0.6705 -0.5868 -0.227>}
 dyn_frontHair_L02 {parent=dyn_frontHair_L01, loc=<4.0623 0 0>, rot=<0 0 0 1>}
 base_backHair_R {parent=head, loc=<2.4693 -2.443 6.7053>, rot=<0.2629 0.5517 0.7866 -0.0882>}
 dyn_backHair_R01 {parent=base_backHair_R, loc=<0 0 0>, rot=<0.0008 -0.0859 0.0001 -0.9963>}
 dyn_backHair_R02 {parent=dyn_backHair_R01, loc=<3.6302 0 0>, rot=<0 0 0 1>}
 base_backHair_C {parent=head, loc=<1.0895 -5.8538 3.0958>, rot=<0.1241 -0.8229 -0.4918 0.2561>}
 dyn_backHair_C01 {parent=base_backHair_C, loc=<0 0 0>, rot=<0 0 0 1>}
 dyn_backHair_C02 {parent=dyn_backHair_C01, loc=<3.7662 0 0>, rot=<0 0 0 1>}
 base_backHair_L {parent=head, loc=<1.5857 -3.3932 -5.4821>, rot=<0.196 -0.8729 -0.4145 -0.167>}
 dyn_backHair_L01 {parent=base_backHair_L, loc=<0 0 0>, rot=<0 0 0 1>}
 dyn_backHair_L02 {parent=dyn_backHair_L01, loc=<3.6478 0 0>, rot=<0 0 0 1>}
 base_card_00 {parent=head, loc=<10.7773 4.1528 7.1213>, rot=<-0.4525 0.2306 0.0071 -0.8614>}
 dyn_card01 {parent=base_card_00, loc=<0 0 0>, rot=<0 0 0 1>}
 dyn_card02 {parent=dyn_card01, loc=<2.4786 0 0>, rot=<0 0 0 1>}
 base_belt_00_l {parent=head, loc=<3.7349 1.5362 -7.7257>, rot=<-0.0029 0.7071 0.7071 -0.0029>}
 dyn_belt_L01 {parent=base_belt_00_l, loc=<0 0 0>, rot=<0 0 0 1>}
 dyn_belt_L02 {parent=dyn_belt_L01, loc=<2.8297 0 0>, rot=<0 0 0 1>}
 dyn_belt_L03 {parent=dyn_belt_L02, loc=<2.7686 0 0>, rot=<0 0 0 1>}
 dyn_belt_L04 {parent=dyn_belt_L03, loc=<3.0778 0 0>, rot=<0 0 0 1>}
 base_belt_00_r {parent=head, loc=<3.7349 1.5362 7.7127>, rot=<-0.0029 0.7071 0.7071 -0.0029>}
 dyn_belt_R01 {parent=base_belt_00_r, loc=<0 0 0>, rot=<0 0 0 1>}
 dyn_belt_R02 {parent=dyn_belt_R01, loc=<2.8297 0 0>, rot=<0 0 0 1>}
 dyn_belt_R03 {parent=dyn_belt_R02, loc=<2.7686 0 0>, rot=<0 0 0 1>}
 dyn_belt_R04 {parent=dyn_belt_R03, loc=<3.0778 0 0>, rot=<0 0 0 1>}
 clavicle_l {parent=spine_05, loc=<5.4831 -0.1053 -3.2946>, rot=<0.0465 -0.7289 -0.009 -0.6829>}
 upperarm_l {parent=clavicle_l, loc=<9.8658 -1.2136 -0.4105>, rot=<0.0164 -0.3245 -0.0269 -0.9454>}
 clavicle_r {parent=spine_05, loc=<5.4831 -0.1053 3.2946>, rot=<0.7289 0.0465 -0.6829 0.009>}
 upperarm_r {parent=clavicle_r, loc=<-9.8658 1.2136 0.4105>, rot=<0.0164 -0.3245 -0.0269 -0.9454>}
]
