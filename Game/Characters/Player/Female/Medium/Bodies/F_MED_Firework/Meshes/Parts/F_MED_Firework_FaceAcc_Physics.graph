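# Physics asset for the Firework female medium character's face accessory mesh, giving physics bodies to hair, belt and card bones.
# Capsules cover head, neck, clavicles, upper arms and the accessory bones.
# Two-link back-hair chains hang from base bones with free rotation on the head, while two-link card and four-link belt chains attach to the head directly.
# Chain joints have narrow swing and twist limits, and both left front-hair bones constrain separately to the head.
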PhysicsAsset /Game/Characters/Player/Female/Medium/Bodies/F_MED_Firework/Meshes/Parts/F_MED_Firework_FaceAcc_Physics
bodies {
 base_backHair_C [{shapeType=Capsule, length=0.5, radius=0.5, shapeName=base_backHair_C_capsule, rotation=<rot 0 -90 0>}]
 base_backHair_L [{shapeType=Capsule, length=0.5, radius=0.5, shapeName=base_backHair_L_capsule, rotation=<rot 0 -90 0>}]
 base_backHair_R [{shapeType=Capsule, length=0.5, radius=0.5, shapeName=base_backHair_R_capsule, rotation=<rot 0 -90 0>}]
 clavicle_l [{shapeType=Capsule, length=2.531, radius=6.469, shapeName=clavicle_l_capsule, center=<3 0 0>, rotation=<rot 90 0 0>}]
 clavicle_r [{shapeType=Capsule, length=9.6001, radius=5.969, shapeName=clavicle_r_capsule, center=<-3 0 0>, rotation=<rot 90 0 0>}]
 dyn_backHair_C01 [{shapeType=Capsule, length=2, radius=2, shapeName=dyn_backHair_C01_capsule, center=<1 0 0>, rotation=<rot 90 0 0>}]
 dyn_backHair_C02 [{shapeType=Capsule, length=7.5153, radius=1.8203, shapeName=dyn_backHair_C02_capsule, center=<1 0 0>, rotation=<rot 90 180 90>}]
 dyn_backHair_L01 [{shapeType=Capsule, length=2, radius=2, shapeName=dyn_backHair_L01_capsule, center=<1 0 0>, rotation=<rot 90 0 0>}]
 dyn_backHair_L02 [{
   shapeType=Capsule
   length=5.2561
   radius=2.1283
   shapeName=dyn_backHair_L02_capsule
   center=<1 0 0>
   rotation=<rot -0.0001 -179.9977 -167.6242>
  }]
 dyn_backHair_R01 [{shapeType=Capsule, length=1.25, radius=1.25, shapeName=dyn_backHair_R01_capsule, center=<1 0 0>, rotation=<rot 90 0 0>}]
 dyn_backHair_R02 [{shapeType=Capsule, length=1.25, radius=1.25, shapeName=dyn_backHair_R02_capsule, center=<1 0 0>, rotation=<rot 90 0 0>}]
 dyn_belt_L01 [{shapeType=Capsule, length=1.25, radius=1.25, shapeName=dyn_belt_L01_capsule, center=<1 0 0>, rotation=<rot 90 0 0>}]
 dyn_belt_L02 [{shapeType=Capsule, length=1.25, radius=1.25, shapeName=dyn_belt_L02_capsule, center=<1 0 0>, rotation=<rot 90 0 0>}]
 dyn_belt_L03 [{shapeType=Capsule, length=1.25, radius=1.25, shapeName=dyn_belt_L03_capsule, center=<1 0 0>, rotation=<rot 90 0 0>}]
 dyn_belt_L04 [{shapeType=Capsule, length=1.25, radius=1.25, shapeName=dyn_belt_L04_capsule, rotation=<rot 90 0 0>}]
 dyn_belt_R01 [{shapeType=Capsule, length=1.0299, radius=1.4701, shapeName=dyn_belt_R01_capsule, center=<1 0 0>, rotation=<rot 90 0 0>}]
 dyn_belt_R02 [{shapeType=Capsule, length=0.9592, radius=1.5408, shapeName=dyn_belt_R02_capsule, center=<1 0 0>, rotation=<rot 90 0 0>}]
 dyn_belt_R03 [{shapeType=Capsule, length=1.25, radius=1.25, shapeName=dyn_belt_R03_capsule, center=<1 0 0>, rotation=<rot 90 0 0>}]
 dyn_belt_R04 [{shapeType=Capsule, length=1.25, radius=1.25, shapeName=dyn_belt_R04_capsule, center=<1 0 0>, rotation=<rot 90 0 0>}]
 dyn_card01 [{shapeType=Capsule, length=1.25, radius=1.25, shapeName=dyn_card01_capsule, center=<2.79 0 0>, rotation=<rot 90 0 -1>}]
 dyn_card02 [{shapeType=Capsule, length=1.25, radius=1.25, shapeName=dyn_card02_capsule, center=<2.79 0 0>, rotation=<rot 90 0 0>}]
 dyn_frontHair_L01 [{shapeType=Capsule, length=1, radius=1, shapeName=dyn_frontHair_L01_capsule, center=<1 0 0>, rotation=<rot 90 0 0>}]
 dyn_frontHair_L02 [{shapeType=Capsule, length=1, radius=1, shapeName=dyn_frontHair_L02_capsule, center=<1 0 0>, rotation=<rot 90 0 0>}]
 head [{shapeType=Capsule, length=2.7212, radius=6.9515, shapeName=head_capsule, center=<1 0 0>, rotation=<rot 90 0 0>}]
 neck_01 [{shapeType=Capsule, length=9.7999, radius=4.0078, shapeName=neck_01_capsule, center=<3 0 0>, rotation=<rot 90 0 0>}]
 upperarm_l [{shapeType=Capsule, length=15, radius=4, shapeName=upperarm_l_capsule, center=<10 0 0>, rotation=<rot 90 0 0>}]
 upperarm_r [{shapeType=Capsule, length=15, radius=3.4, shapeName=upperarm_r_capsule, center=<-9.7056 0 0>, rotation=<rot 90 0 0>}]
}
constraints [
 {bone1Name=dyn_backHair_C02, bone2Name=dyn_backHair_C01, swing1LimitDegrees=1, swing2LimitDegrees=20, twistLimitDegrees=1}
 {bone1Name=dyn_backHair_L02, bone2Name=dyn_backHair_L01, swing1LimitDegrees=7, swing2LimitDegrees=1, twistLimitDegrees=1}
 {bone1Name=dyn_backHair_R02, bone2Name=dyn_backHair_R01, swing1LimitDegrees=12, swing2LimitDegrees=1, twistLimitDegrees=1}
 {bone1Name=dyn_card02, bone2Name=dyn_card01, swing1LimitDegrees=8, swing2LimitDegrees=1, twistLimitDegrees=1}
 {bone1Name=dyn_card01, bone2Name=head, swing1LimitDegrees=7, swing2LimitDegrees=1, twistLimitDegrees=1}
 {bone1Name=dyn_frontHair_L01, bone2Name=head, swing1LimitDegrees=7, swing2LimitDegrees=1, twistLimitDegrees=1}
 {bone1Name=dyn_frontHair_L02, bone2Name=head, swing1LimitDegrees=7, swing2LimitDegrees=1, twistLimitDegrees=1}
 {bone1Name=dyn_backHair_C01, bone2Name=base_backHair_C, swing1LimitDegrees=1, swing2LimitDegrees=20, twistLimitDegrees=1}
 {bone1Name=dyn_backHair_R01, bone2Name=base_backHair_R, swing1LimitDegrees=10, swing2LimitDegrees=1, twistLimitDegrees=1}
 {bone1Name=dyn_backHair_L01, bone2Name=base_backHair_L, swing1LimitDegrees=12, swing2LimitDegrees=1, twistLimitDegrees=1}
 {bone1Name=upperarm_l, bone2Name=clavicle_l, swing1LimitDegrees=45, swing2LimitDegrees=45, twistLimitDegrees=1}
 {bone1Name=upperarm_r, bone2Name=clavicle_r, swing1LimitDegrees=45, swing2LimitDegrees=45, twistLimitDegrees=1}
 {bone1Name=dyn_belt_R01, bone2Name=head, swing1LimitDegrees=10, swing2LimitDegrees=10, twistLimitDegrees=1}
 {bone1Name=dyn_belt_R02, bone2Name=dyn_belt_R01, swing1LimitDegrees=7, swing2LimitDegrees=10, twistLimitDegrees=1}
 {bone1Name=dyn_belt_R03, bone2Name=dyn_belt_R02, swing1LimitDegrees=7.5, swing2LimitDegrees=10, twistLimitDegrees=1}
 {bone1Name=dyn_belt_R04, bone2Name=dyn_belt_R03, swing1LimitDegrees=10, swing2LimitDegrees=10, twistLimitDegrees=1}
 {bone1Name=dyn_belt_L01, bone2Name=head, swing1LimitDegrees=10, swing2LimitDegrees=10, twistLimitDegrees=1}
 {bone1Name=dyn_belt_L02, bone2Name=dyn_belt_L01, swing1LimitDegrees=10, swing2LimitDegrees=10, twistLimitDegrees=1}
 {bone1Name=dyn_belt_L03, bone2Name=dyn_belt_L02, swing1LimitDegrees=10, swing2LimitDegrees=10, twistLimitDegrees=1}
 {bone1Name=dyn_belt_L04, bone2Name=dyn_belt_L03, swing1LimitDegrees=10, swing2LimitDegrees=10, twistLimitDegrees=1}
 {bone1Name=head, bone2Name=neck_01, swing1LimitDegrees=10, swing2LimitDegrees=10, twistLimitDegrees=1}
 {
  bone1Name=base_backHair_C
  bone2Name=head
  swing1LimitDegrees=45
  swing1Motion=Free
  swing2LimitDegrees=45
  swing2Motion=Free
  twistLimitDegrees=45
  twistMotion=Free
 }
 {
  bone1Name=base_backHair_L
  bone2Name=head
  swing1LimitDegrees=45
  swing1Motion=Free
  swing2LimitDegrees=45
  swing2Motion=Free
  twistLimitDegrees=45
  twistMotion=Free
 }
 {
  bone1Name=base_backHair_R
  bone2Name=head
  swing1LimitDegrees=45
  swing1Motion=Free
  swing2LimitDegrees=45
  swing2Motion=Free
  twistLimitDegrees=45
  twistMotion=Free
 }
]
